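# Anim Blueprint for the Terry Broke head cosmetic, adding spring-driven ear jiggle and adjusting head rotation.
# It layers the Body mesh's pose from the clavicles onto the attached parent's pose, runs AnimDynamics on both ears, then caches the result as 'collar'.
# The TerryBrokeVampire metatag, checked on init or customization load, selects between a fixed additive head offset and a per-axis scaled counter-rotation computed each update from head-to-neck_02 delta.
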AnimBlueprint /BRCosmetics/Characters/Player/Male/Medium/Heads/M_MED_TerryBroke_Head/Meshes/TerryBroke_Head_Skeleton_AnimBlueprint
target_skeleton=&/M_MED_TerryBroke_Head_Skeleton
graphs [
 AnimGraph {nodes [
   n1 = Anim Dynamics - Bone: ear_l!(SimulationSpace:AnimPhysSimSpaceType Enum=Component, ComponentPose, Alpha=1, ExternalForce="0,0,0") -> (Pose) {
    angularDampingOverride=0.9999
    angularSpringConstant=1111
    bAngularSpring=true
    bOverrideAngularDamping=true
    bUseGravityOverride=true
    linearDampingOverride=0.9999
    simulationSpace=World
    alphaBoolBlend {blendOption=Linear}
    boundBone {boneName=ear_l}
    componentAppliedLinearAccClamp=<2222 2222 2222>
    componentLinearAccScale {z=0.1}
    componentLinearVelScale {z=0.1}
    physicsBodyDefinitions [{
      boundBone {boneName=ear_l}
      boxExtents=<9 9 9>
      constraintSetup {angularLimitsMax=<10 10 10>, angularLimitsMin=<-5 -5 -5>, angularTarget=<1 0 0>}
      localJointOffset=<0.2 0 0>
     }]
    simSpaceSettings {maxAngularAcceleration=300, maxAngularVelocity=50, simSpaceAngularAlpha=0.4}
   }
   Anim Dynamics - Bone: ear_r!(SimulationSpace:AnimPhysSimSpaceType Enum=Component, ComponentPose=@n1, Alpha=1, ExternalForce="0,0,0") -> (Pose) {
    angularDampingOverride=0.9999
    angularSpringConstant=1111
    bAngularSpring=true
    bOverrideAngularDamping=true
    bUseGravityOverride=true
    linearDampingOverride=0.9999
    simulationSpace=World
    alphaBoolBlend {blendOption=Linear}
    boundBone {boneName=ear_r}
    componentAppliedLinearAccClamp=<2222 2222 2222>
    componentLinearAccScale {z=0.1}
    componentLinearVelScale {z=0.1}
    physicsBodyDefinitions [{
      boundBone {boneName=ear_r}
      boxExtents=<9 9 9>
      constraintSetup {angularLimitsMax=<5 5 5>, angularLimitsMin=<-5 -5 -5>, angularTarget=<1 0 0>}
      localJointOffset=<-0.2 0 0>
     }]
    simSpaceSettings {maxAngularAcceleration=300, maxAngularVelocity=50, simSpaceAngularAlpha=0.4}
   }
   Input Pose()
   n2 = "Transform (Modify) Bone - Bone: head"(
    ComponentPose=Local To Component(LocalPose=Use cached pose 'collar'())
    Alpha=Get headlock() -> (headlock:double)
    Translation="0,0,0"
    Rotation="-23.800,-103.114,25.311"
    Scale="1,1,1"
   ) {rotationMode=BMM_Additive, rotationSpace=BCS_ParentBoneSpace, boneToModify {boneName=head}, rotation=<rot -23.8 -103.114 25.311>}
   n3 = "Transform (Modify) Bone - Bone: head"(
    ComponentPose=Local To Component(LocalPose=Use cached pose 'collar'())
    Alpha=1
    Translation="0,0,0"
    Rotation=Get headRotLimit_no_colar() -> (headRotLimit_no_colar:Rotator)
    Scale="1,1,1"
   ) {rotationMode=BMM_Additive, rotationSpace=BCS_ParentBoneSpace, boneToModify {boneName=head}}
   Output Pose(Result=Blend Poses by bool(
     BlendPose_0=Component To Local(ComponentPose=@n3)
     BlendPose_1=Component To Local(ComponentPose=@n2)
     bActiveValue=Get block head() -> (block head:bool)
    ))
   n4 = Copy Pose From Mesh(SourceMeshComponent=Get Body() -> (Body:Skeletal Mesh Component Object Reference)) {bUseAttachedParent=true}
   n5 = Copy Pose From Mesh() {bUseAttachedParent=true}
   n6 = Anim Dynamics - Bone: ear_r!(
    ComponentPose=Local To Component(LocalPose=Layered blend per bone(BasePose=@n5, BlendPoses_0=@n4) {layerSetup [{branchFilters [{boneName=clavicle_l}, {boneName=clavicle_r}]}]})
    Alpha=1
    ExternalForce="0,0,0"
   ) -> (Pose) {
    angularDampingOverride=0.99
    angularSpringConstant=1500
    bAngularSpring=true
    bOverrideAngularDamping=true
    bUseGravityOverride=true
    simulationSpace=RootRelative
    boundBone {boneName=ear_r}
    componentAppliedLinearAccClamp=<3000 3000 10000>
    componentLinearAccScale=<0.3 0.3 0.5>
    componentLinearVelScale=<0.3 0.3 1>
    gravityOverride {z=-980}
    physicsBodyDefinitions [{
      boundBone {boneName=ear_r}
      boxExtents=<13 13 13>
      constraintSetup {angularLimitsMax=<10 10 10>, angularLimitsMin=<-10 -10 -10>, angularTarget=<1 0 0>}
      localJointOffset=<-0.2 0 0>
     }]
    relativeSpaceBone {boneName=spine_05}
    simSpaceSettings {maxAngularAcceleration=40, maxAngularVelocity=8, simSpaceAngularAlpha=0.3}
   }
   n7 = Anim Dynamics - Bone: ear_l!(ComponentPose=@n6, Alpha=1, ExternalForce="0,0,0") -> (Pose) {
    angularDampingOverride=0.999
    angularSpringConstant=1500
    bAngularSpring=true
    bOverrideAngularDamping=true
    bUseGravityOverride=true
    simulationSpace=RootRelative
    boundBone {boneName=ear_l}
    componentAppliedLinearAccClamp=<3000 3000 10000>
    componentLinearAccScale=<0.3 0.3 0.5>
    componentLinearVelScale=<0.3 0.3 1>
    gravityOverride {z=-980}
    physicsBodyDefinitions [{
      boundBone {boneName=ear_l}
      boxExtents=<11 11 11>
      constraintSetup {angularLimitsMax=<10 10 10>, angularLimitsMin=<-10 -10 -10>, angularTarget=<1 0 0>}
      localJointOffset=<0.2 0 0>
     }]
    relativeSpaceBone {boneName=spine_05}
    simSpaceSettings {maxAngularAcceleration=40, maxAngularVelocity=8, simSpaceAngularAlpha=0.3}
   }
   Save cached pose 'collar'(Pose=Component To Local(ComponentPose=@n7)) {cacheName=collar}
  ]}
 EventGraph {nodes [
   Event BlueprintUpdateAnimation()
   > n1 = Sequence()
   n2 = GetOwnerSkeletalMeshComponent()
   n3 = BreakRotator(InRot=BreakTransform(InTransform=GetDeltaTransformFromRefPose(self=@n2, BoneName=head, BaseName=neck_02)).Rotation)
   n4 = Get Z percentageEmote() -> (Z percentageEmote:float)
   n5 = float * float!(A=@n3, B=-1, C:double=Get X percentageEmote() -> (X percentageEmote:float)) -> (ReturnValue)
   n6 = float * float(A=@n3.Pitch, B=-1, C=Get Y percentageEmote() -> (Y percentageEmote:float))
   Set headRotLimit(execute=@n1.then_1, headRotLimit:Rotator=MakeRotator(Roll=@n5, Pitch=@n6, Yaw=float * float(A=@n3.Yaw, B=-1, C=@n4))) -> (Output_Get:Rotator)
   n7 = Get Z percentageEmote_no_colar() -> (Z percentageEmote_no_colar:float)
   > Set headRotLimit_no_colar(headRotLimit_no_colar:Rotator=MakeRotator(Roll=@n5, Pitch=@n6, Yaw=float * float(A=@n3.Yaw, B=-1, C=@n7))) -> (Output_Get:Rotator)
   n8 = Event OnCharacterCustomizationLoaded()
   Is Valid(exec=[@n8.then, @n13.then], InputObject=@n8.CustomBaseMesh)
   > Set Body(Body:Skeletal Mesh Component Object Reference=@n8.CustomBaseMesh) -> (Output_Get:Skeletal Mesh Component Object Reference)
   n9 = GetPartSkeletalMeshComponent()
   > Is Valid(InputObject=@n9)
   n10 = CharacterItemDefinitionHasAnyMetaTag(Tags=Make Array("[0]"=(TagName="Cosmetics.Metatags.TerryBrokeVampire")))
   > Set block head(block head:bool=@n10) -> (Output_Get:bool)
   > n11 = Get Dynamic Material Instance(SkeletalMeshComponent=@n9)
   > n12 = Branch(Condition=@n10)
   > Set headlock(headlock:double=0) -> (Output_Get:double)
   Set headlock(execute=@n12.else, headlock=1)
   n13 = Event BlueprintInitializeAnimation()
   Reroute Node(InputPin:Material Instance Dynamic Object Reference=@n11) -> (OutputPin:Material Instance Dynamic Object Reference)
  ]}
]
variables [
 block head {type=boolean}
 Body {type=SkeletalMeshComponent}
 headlock {type=number}
 headRotLimit {type=Rotator}
 headRotLimit_no_colar {type=Rotator}
 X percentageEmote {type=number, default=1}
 Y percentageEmote {type=number, default=1}
 Z percentageEmote {type=number, default=1}
 Z percentageEmote_no_colar {type=number, default=1}
]
